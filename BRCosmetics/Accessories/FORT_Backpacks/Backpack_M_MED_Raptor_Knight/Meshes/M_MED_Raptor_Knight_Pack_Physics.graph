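# Physics asset for the Raptor Knight medium male backpack cosmetic, covering its dangling axe, hanging strap and split arrow.
# Each of these dyn_ joints has a capsule body constrained directly to the capsule on the pack bone, forming a star rather than a chain.
# The strap and arrow have angular swing and twist limits, while the axe locks swing1 and twist, leaving swing2 as its only angular swing axis.
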PhysicsAsset /BRCosmetics/Accessories/FORT_Backpacks/Backpack_M_MED_Raptor_Knight/Meshes/M_MED_Raptor_Knight_Pack_Physics
bodies {
 dyn_axe_jnt [{shapeType=Capsule, length=30, radius=2, center=<1.3924 0 0>, rotation=<rot 90 0 0>}]
 dyn_hangingStrap_jnt [{shapeType=Capsule, length=7, radius=1, center=<0.6436 0 0>, rotation=<rot 90 0 0>}]
 dyn_splitArrow_jnt [{shapeType=Capsule, length=10, radius=0.8, center=<-4.0323 0.7808 0.6487>, rotation=<rot 79.9525 -6.8867 0>}]
 pack [{shapeType=Capsule, length=3, radius=7}]
}
constraints [
 {bone1Name=dyn_hangingStrap_jnt, bone2Name=pack, swing1LimitDegrees=15, swing2LimitDegrees=40, twistLimitDegrees=10}
 {bone1Name=dyn_splitArrow_jnt, bone2Name=pack, swing1LimitDegrees=50, swing2LimitDegrees=20, twistLimitDegrees=20}
 {
  bone1Name=dyn_axe_jnt
  bone2Name=pack
  swing1LimitDegrees=45
  swing1Motion=Locked
  swing2LimitDegrees=40
  twistLimitDegrees=45
  twistMotion=Locked
 }
]
