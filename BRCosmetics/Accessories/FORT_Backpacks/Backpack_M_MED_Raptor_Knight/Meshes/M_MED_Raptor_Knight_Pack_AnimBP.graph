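# Animation Blueprint for the Raptor Knight backpack cosmetic, adding secondary physics motion to the pack.
# A RigidBody node simulates its physics asset in root-bone space with overridden gravity, lowering linear acceleration response while crouching.
# Additive bone transforms then nudge and tilt the pack bone, and rotate the hanging strap joint only while skydiving.
# The event graph holds only unwired stub nodes.
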AnimBlueprint /BRCosmetics/Accessories/FORT_Backpacks/Backpack_M_MED_Raptor_Knight/Meshes/M_MED_Raptor_Knight_Pack_AnimBP
target_skeleton=&/M_MED_Raptor_Knight_Pack_Skeleton
graphs [
 AnimGraph {nodes [
   n1 = RigidBody!(
    ComponentPose=Local To Component(LocalPose=Input Pose())
    Alpha
    ExternalForce
    ComponentLinearAccScale:Vector=SelectVector(A="0.23,0.23,0.23", B="0.5,0.5,0.5", bPickA=Get bIsCrouching())
   ) -> (Pose) {
    bClampLinearTranslationLimitToRefPose=true
    bOverrideWorldGravity=true
    overridePhysicsAsset=&/M_MED_Raptor_Knight_Pack_Physics
    simulationSpace=BaseBoneSpace
    alphaBoolBlend {blendOption=Linear}
    baseBoneRef {boneName=root}
    componentAppliedLinearAccClamp=<4000 4000 4000>
    componentLinearAccScale=<0.5 0.5 0.5>
    componentLinearVelScale=<0.3 0.3 0.3>
    overrideWorldGravity {z=-500}
   }
   n2 = "Transform (Modify) Bone - Bone: pack"(ComponentPose=@n1, Translation="0,-0.4,0", Rotation="0,0,6") {
    rotationMode=BMM_Additive
    translationMode=BMM_Additive
    alphaBoolBlend {blendOption=Linear}
    boneToModify {boneName=pack}
    rotation {roll=6}
    translation {y=-0.4}
   }
   n3 = "Transform (Modify) Bone - Bone: dyn_hangingStrap_jnt"(ComponentPose=@n2, Alpha="To Float (Boolean)"(InBool=Get bIsSkyDiving()), Translation="0,-0.4,0", Rotation="0,0,15", Scale="1,1,1") {
    rotationMode=BMM_Additive
    alphaBoolBlend {blendOption=Linear}
    boneToModify {boneName=dyn_hangingStrap_jnt}
    rotation {roll=15}
    translation {y=-0.4}
   }
   Output Pose(Result=Component To Local(ComponentPose=@n3)) {layerGroup=None}
  ]}
 EventGraph {nodes [
   TryGetPawnOwner()
   Event BlueprintUpdateAnimation()
  ]}
]
variables []
